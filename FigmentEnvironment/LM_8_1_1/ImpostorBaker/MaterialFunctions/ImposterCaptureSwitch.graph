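# Material function that lets a material render its individual attributes for impostor baking, behind a static switch that otherwise passes attributes through.
# MPC_ImposterCapture weights select base color, specular, roughness, subsurface color, remapped normal or depth, summed into emissive.
# EnableCapture fades out the original shading inputs, while CustomLighting is added to base color.
# Depth comes from world position transformed into the capture basis around Origin, scaled by radius R and inverted.
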MaterialFunction /FigmentEnvironment/LM_8_1_1/ImpostorBaker/MaterialFunctions/ImposterCaptureSwitch
graph {
 properties {description=Used for rendering Imposors}
 nodes [
  Constant3Vector() {constant=<rgba 1 0 0 0>}
  Constant3Vector() {constant=<rgba 0 0 1 0>}
  n1 = FunctionInput() {inputName=Material Attributes, inputType=FunctionInput_MaterialAttributes}
  n2 = BreakMaterialAttributes!(Attr=@n1)
  n3 = OneMinus(None=CollectionParameter() {collection=/FigmentEnvironment/LM_8_1_1/ImpostorBaker/MPC/MPC_ImposterCapture, parameterName=EnableCapture})
  n4 = Add(
   A=CollectionParameter() {collection=&MPC_ImposterCapture, parameterName=CustomLighting}
   B=Multiply(A=@n2.BaseColor, B=@n3)
  )
  n5 = Multiply(A=@n2.Metallic, B=@n3)
  n6 = Multiply(A=@n2.Specular, B=@n3)
  n7 = Multiply(A=@n2.Roughness, B=@n3)
  n8 = BreakMaterialAttributes(Attr=@n1)
  n9 = Multiply(
   A=CollectionParameter() {collection=&MPC_ImposterCapture, parameterName=BaseColor}
   B=@n8.BaseColor
  )
  n10 = Multiply(
   A=CollectionParameter() {collection=&MPC_ImposterCapture, parameterName=Specular}
   B=@n8.Specular
  )
  n11 = CollectionParameter() {collection=&MPC_ImposterCapture, parameterName=Roughness}
  n12 = Add(
   A=Add(A=Add(A=@n9, B=@n10), B=Multiply(A=@n11, B=@n8.Roughness))
   B=Multiply(
    A=CollectionParameter() {collection=&MPC_ImposterCapture, parameterName=SubSurface}
    B=@n8.SubsurfaceColor
   )
  )
  n13 = Normalize(VectorInput=@n8.Normal)
  n14 = Multiply(
   A=CollectionParameter() {collection=&MPC_ImposterCapture, parameterName=Normal}
   B=ConstantBiasScale(None=@n13)
  )
  n15 = Subtract(
   A=WorldPosition().XYZ
   B=ComponentMask(None=CollectionParameter() {collection=&MPC_ImposterCapture, parameterName=Origin}) {b=true, g=true, r=true}
  )
  n16 = ComponentMask(None=CollectionParameter() {collection=&MPC_ImposterCapture, parameterName=X}) {b=true, g=true, r=true}
  n17 = ComponentMask(None=CollectionParameter() {collection=&MPC_ImposterCapture, parameterName=Y}) {b=true, g=true, r=true}
  n18 = ComponentMask(None=CollectionParameter() {collection=&MPC_ImposterCapture, parameterName=Z}) {b=true, g=true, r=true}
  n19 = ConstantBiasScale(None=Divide(
    A=MaterialFunctionCall!(VectorToTransform=@n15, BasisX=@n16, BasisY=@n17, BasisZ=@n18) {materialFunction=/Engine/Functions/Engine_MaterialFunctions02/Math/InverseTransformMatrix}.Transformed Vector 3
    B=CollectionParameter() {collection=&MPC_ImposterCapture, parameterName=R}
   ))
  n20 = OneMinus(None=Saturate(None=ComponentMask(None=@n19) {b=true}))
  n21 = CollectionParameter() {collection=&MPC_ImposterCapture, parameterName=Depth}
  n22 = Add(A=Multiply(A=@n2.EmissiveColor, B=@n3), B=Add(A=Add(A=@n12, B=@n14), B=Multiply(A=@n20, B=@n21)))
  n23 = Multiply(A=@n2.SubsurfaceColor, B=@n3)
  n24 = Multiply(A=@n2.ClearCoat, B=@n3)
  FunctionOutput(None=StaticSwitchParameter(
    True=MakeMaterialAttributes!(
     BaseColor=@n4
     Metallic=@n5
     Specular=@n6
     Roughness=@n7
     Anisotropy
     EmissiveColor=@n22
     Opacity=@n2.Opacity
     OpacityMask=@n2.Opacity
     Normal=@n2.Opacity
     Tangent
     WorldPositionOffset
     SubsurfaceColor=@n23
     ClearCoat=@n24
     ClearCoatRoughness
     AmbientOcclusion=@n2.Opacity
     Refraction
     CustomizedUVs_0=@n2.Opacity
     CustomizedUVs_1=@n2.Opacity
     CustomizedUVs_2=@n2.Opacity
     CustomizedUVs_3=@n2.Opacity
     CustomizedUVs_4=@n2.Opacity
     CustomizedUVs_5=@n2.Opacity
     CustomizedUVs_6=@n2.Opacity
     CustomizedUVs_7=@n2.Opacity
     PixelDepthOffset=@n2.Opacity
     ShadingModel
     Displacement
    )
    False=@n1
   ) {defaultValue=true, parameterName=Enable Capture Function})
  Transform(None=@n13)
 ]
}
